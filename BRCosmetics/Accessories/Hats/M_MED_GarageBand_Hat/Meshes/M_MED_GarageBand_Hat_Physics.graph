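# Physics asset for the GarageBand cosmetic hat, covering its dynamic hair strands, front and back cards, and feather.
# Capsule bodies form short chains constrained to the head with limited swing and tight twist.
# Hair groups A, B and C branch from their second bone into three tips.
# Head, neck_02 and spine_05 also carry capsules, linked by constraints with free swing and twist.
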PhysicsAsset /BRCosmetics/Accessories/Hats/M_MED_GarageBand_Hat/Meshes/M_MED_GarageBand_Hat_Physics
bodies {
 dyn_GarageBand_card_back [{shapeType=Capsule, length=2, radius=2, rotation=<rot 90 0 0>}]
 dyn_GarageBand_card_front [{shapeType=Capsule, length=2, radius=2, rotation=<rot 90 0 0>}]
 dyn_GarageBand_feather_01 [{shapeType=Capsule, length=2, radius=2, rotation=<rot 90 0 0>}]
 dyn_GarageBand_feather_02 [{shapeType=Capsule, length=2, radius=2, rotation=<rot 90 0 0>}]
 dyn_GarageBand_hair_A_01 [{shapeType=Capsule, length=2, radius=2, center=<2.5 0 0>, rotation=<rot 90 -90 -79.9999>}]
 dyn_GarageBand_hair_A_02 [{shapeType=Capsule, length=2, radius=2, center=<2.5 0 0>, rotation=<rot 90 0 0>}]
 dyn_GarageBand_hair_A_03 [{shapeType=Capsule, length=7.25, radius=2, center=<2.5 0 0>, rotation=<rot 90 0 0>}]
 dyn_GarageBand_hair_A_04 [{shapeType=Capsule, length=7.25, radius=2, center=<2.5 0 0>, rotation=<rot 90 0 0>}]
 dyn_GarageBand_hair_A_05 [{shapeType=Capsule, length=7.25, radius=2, center=<2.5 0 0>, rotation=<rot 90 0 0>}]
 dyn_GarageBand_hair_B_01 [{shapeType=Capsule, length=2, radius=2, rotation=<rot 90 0 0>}]
 dyn_GarageBand_hair_B_02 [{shapeType=Capsule, length=2, radius=2, rotation=<rot 90 0 0>}]
 dyn_GarageBand_hair_B_03 [{shapeType=Capsule, length=2, radius=2, rotation=<rot 90 0 0>}]
 dyn_GarageBand_hair_B_04 [{shapeType=Capsule, length=2, radius=2, rotation=<rot 90 0 0>}]
 dyn_GarageBand_hair_B_05 [{shapeType=Capsule, length=2, radius=2, rotation=<rot 90 0 0>}]
 dyn_GarageBand_hair_C_01 [{shapeType=Capsule, length=2, radius=2, rotation=<rot 90 0 0>}]
 dyn_GarageBand_hair_C_02 [{shapeType=Capsule, length=2, radius=2, rotation=<rot 90 0 0>}]
 dyn_GarageBand_hair_C_03 [{shapeType=Capsule, length=2, radius=2, rotation=<rot 90 0 0>}]
 dyn_GarageBand_hair_C_04 [{shapeType=Capsule, length=2, radius=2, rotation=<rot 90 0 0>}]
 dyn_GarageBand_hair_C_05 [{shapeType=Capsule, length=2, radius=2, rotation=<rot 90 0 0>}]
 dyn_GarageBand_hair_E_01 [{shapeType=Capsule, length=2, radius=2, rotation=<rot 90 0 0>}]
 dyn_GarageBand_hair_E_02 [{shapeType=Capsule, length=2, radius=2, rotation=<rot 90 0 0>}]
 dyn_GarageBand_hair_E_03 [{shapeType=Capsule, length=2, radius=2, rotation=<rot 90 0 0>}]
 dyn_GarageBand_hair_F_D_01 [{shapeType=Capsule, length=2, radius=2, rotation=<rot 90 0 0>}]
 dyn_GarageBand_hair_F_D_02 [{shapeType=Capsule, length=2, radius=2, rotation=<rot 90 0 0>}]
 dyn_GarageBand_hair_F_D_03 [{shapeType=Capsule, length=2, radius=2, rotation=<rot 90 0 0>}]
 dyn_GarageBand_hair_G_01 [{shapeType=Capsule, length=2, radius=2, rotation=<rot 90 0 0>}]
 dyn_GarageBand_hair_G_02 [{shapeType=Capsule, length=2, radius=2, rotation=<rot 90 0 0>}]
 dyn_GarageBand_hair_G_03 [{shapeType=Capsule, length=2, radius=2, rotation=<rot 90 0 0>}]
 dyn_GarageBand_hair_H_01 [{shapeType=Capsule, length=2, radius=2, rotation=<rot 90 0 0>}]
 dyn_GarageBand_hair_H_02 [{shapeType=Capsule, length=2, radius=2, rotation=<rot 90 0 0>}]
 dyn_GarageBand_hair_H_03 [{shapeType=Capsule, length=2, radius=2, rotation=<rot 90 0 0>}]
 dyn_GarageBand_hair_I_01 [{shapeType=Capsule, length=2, radius=2, rotation=<rot 90 0 0>}]
 dyn_GarageBand_hair_I_02 [{shapeType=Capsule, length=2, radius=2, rotation=<rot 90 0 0>}]
 dyn_GarageBand_hair_J_01 [{shapeType=Capsule, length=2, radius=2, rotation=<rot 90 0 0>}]
 dyn_GarageBand_hair_J_02 [{shapeType=Capsule, length=2, radius=2, rotation=<rot 90 0 0>}]
 dyn_GarageBand_hair_K_01 [{shapeType=Capsule, length=2, radius=2, rotation=<rot 90 0 0>}]
 head [{shapeType=Capsule, length=2, radius=2, rotation=<rot 90 0 0>}]
 neck_02 [{shapeType=Capsule, length=5, radius=5, center=<0.2975 -0.0569 0.0001>, rotation=<rot 0 -90 0>}]
 spine_05 [
  {shapeType=Capsule, length=6.755, radius=11.755, center=<4.2621 4.3617 -0.0002>, rotation=<rot 0 -90 0>}
  {shapeType=Capsule, length=6.755, radius=11.755, center=<-0.7087 -0.6175 0.0001>, rotation=<rot 0 -90 0>}
 ]
}
constraints [
 {bone1Name=dyn_GarageBand_hair_A_01, bone2Name=head, swing1LimitDegrees=45, swing2LimitDegrees=45, twistLimitDegrees=5}
 {bone1Name=dyn_GarageBand_hair_A_02, bone2Name=dyn_GarageBand_hair_A_01, swing1LimitDegrees=45, swing2LimitDegrees=45, twistLimitDegrees=5}
 {bone1Name=dyn_GarageBand_hair_A_03, bone2Name=dyn_GarageBand_hair_A_02, swing1LimitDegrees=45, swing2LimitDegrees=45, twistLimitDegrees=5}
 {bone1Name=dyn_GarageBand_hair_A_04, bone2Name=dyn_GarageBand_hair_A_02, swing1LimitDegrees=45, swing2LimitDegrees=45, twistLimitDegrees=5}
 {bone1Name=dyn_GarageBand_hair_A_05, bone2Name=dyn_GarageBand_hair_A_02, swing1LimitDegrees=45, swing2LimitDegrees=45, twistLimitDegrees=5}
 {bone1Name=dyn_GarageBand_card_front, bone2Name=head, swing1LimitDegrees=45, swing2LimitDegrees=45, twistLimitDegrees=5}
 {bone1Name=dyn_GarageBand_hair_C_01, bone2Name=head, swing1LimitDegrees=45, swing2LimitDegrees=45, twistLimitDegrees=5}
 {bone1Name=dyn_GarageBand_hair_C_02, bone2Name=dyn_GarageBand_hair_C_01, swing1LimitDegrees=45, swing2LimitDegrees=45, twistLimitDegrees=5}
 {bone1Name=dyn_GarageBand_hair_C_03, bone2Name=dyn_GarageBand_hair_C_02, swing1LimitDegrees=45, swing2LimitDegrees=45, twistLimitDegrees=5}
 {bone1Name=dyn_GarageBand_hair_C_04, bone2Name=dyn_GarageBand_hair_C_02, swing1LimitDegrees=45, swing2LimitDegrees=45, twistLimitDegrees=5}
 {bone1Name=dyn_GarageBand_hair_C_05, bone2Name=dyn_GarageBand_hair_C_02, swing1LimitDegrees=45, swing2LimitDegrees=45, twistLimitDegrees=5}
 {bone1Name=dyn_GarageBand_card_back, bone2Name=head, swing1LimitDegrees=45, swing2LimitDegrees=45, twistLimitDegrees=5}
 {bone1Name=dyn_GarageBand_feather_01, bone2Name=head, swing1LimitDegrees=45, swing2LimitDegrees=45, twistLimitDegrees=5}
 {bone1Name=dyn_GarageBand_feather_02, bone2Name=dyn_GarageBand_feather_01, swing1LimitDegrees=45, swing2LimitDegrees=45, twistLimitDegrees=5}
 {bone1Name=dyn_GarageBand_hair_B_01, bone2Name=head, swing1LimitDegrees=45, swing2LimitDegrees=45, twistLimitDegrees=5}
 {bone1Name=dyn_GarageBand_hair_B_02, bone2Name=dyn_GarageBand_hair_B_01, swing1LimitDegrees=45, swing2LimitDegrees=45, twistLimitDegrees=5}
 {bone1Name=dyn_GarageBand_hair_B_03, bone2Name=dyn_GarageBand_hair_B_02, swing1LimitDegrees=45, swing2LimitDegrees=45, twistLimitDegrees=5}
 {bone1Name=dyn_GarageBand_hair_B_04, bone2Name=dyn_GarageBand_hair_B_02, swing1LimitDegrees=45, swing2LimitDegrees=45, twistLimitDegrees=5}
 {bone1Name=dyn_GarageBand_hair_B_05, bone2Name=dyn_GarageBand_hair_B_02, swing1LimitDegrees=45, swing2LimitDegrees=45, twistLimitDegrees=5}
 {bone1Name=dyn_GarageBand_hair_F_D_01, bone2Name=head, swing1LimitDegrees=45, swing2LimitDegrees=45, twistLimitDegrees=5}
 {
  bone1Name=dyn_GarageBand_hair_F_D_02
  bone2Name=dyn_GarageBand_hair_F_D_01
  swing1LimitDegrees=45
  swing2LimitDegrees=45
  twistLimitDegrees=5
 }
 {
  bone1Name=dyn_GarageBand_hair_F_D_03
  bone2Name=dyn_GarageBand_hair_F_D_02
  swing1LimitDegrees=45
  swing2LimitDegrees=45
  twistLimitDegrees=5
 }
 {bone1Name=dyn_GarageBand_hair_H_01, bone2Name=head, swing1LimitDegrees=45, swing2LimitDegrees=45, twistLimitDegrees=5}
 {bone1Name=dyn_GarageBand_hair_H_02, bone2Name=dyn_GarageBand_hair_H_01, swing1LimitDegrees=45, swing2LimitDegrees=45, twistLimitDegrees=5}
 {bone1Name=dyn_GarageBand_hair_H_03, bone2Name=dyn_GarageBand_hair_H_02, swing1LimitDegrees=45, swing2LimitDegrees=45, twistLimitDegrees=5}
 {bone1Name=dyn_GarageBand_hair_E_01, bone2Name=head, swing1LimitDegrees=45, swing2LimitDegrees=45, twistLimitDegrees=5}
 {bone1Name=dyn_GarageBand_hair_E_02, bone2Name=dyn_GarageBand_hair_E_01, swing1LimitDegrees=45, swing2LimitDegrees=45, twistLimitDegrees=5}
 {bone1Name=dyn_GarageBand_hair_E_03, bone2Name=dyn_GarageBand_hair_E_02, swing1LimitDegrees=45, swing2LimitDegrees=45, twistLimitDegrees=5}
 {bone1Name=dyn_GarageBand_hair_G_01, bone2Name=head, swing1LimitDegrees=45, swing2LimitDegrees=45, twistLimitDegrees=5}
 {bone1Name=dyn_GarageBand_hair_G_02, bone2Name=dyn_GarageBand_hair_G_01, swing1LimitDegrees=45, swing2LimitDegrees=45, twistLimitDegrees=5}
 {bone1Name=dyn_GarageBand_hair_G_03, bone2Name=dyn_GarageBand_hair_G_02, swing1LimitDegrees=45, swing2LimitDegrees=45, twistLimitDegrees=5}
 {bone1Name=dyn_GarageBand_hair_I_01, bone2Name=head, swing1LimitDegrees=45, swing2LimitDegrees=45, twistLimitDegrees=5}
 {bone1Name=dyn_GarageBand_hair_I_02, bone2Name=dyn_GarageBand_hair_I_01, swing1LimitDegrees=45, swing2LimitDegrees=45, twistLimitDegrees=5}
 {bone1Name=dyn_GarageBand_hair_J_01, bone2Name=head, swing1LimitDegrees=45, swing2LimitDegrees=45, twistLimitDegrees=5}
 {bone1Name=dyn_GarageBand_hair_J_02, bone2Name=dyn_GarageBand_hair_J_01, swing1LimitDegrees=45, swing2LimitDegrees=45, twistLimitDegrees=5}
 {bone1Name=dyn_GarageBand_hair_K_01, bone2Name=head, swing1LimitDegrees=45, swing2LimitDegrees=45, twistLimitDegrees=5}
 {
  bone1Name=head
  bone2Name=neck_02
  swing1LimitDegrees=45
  swing1Motion=Free
  swing2LimitDegrees=45
  swing2Motion=Free
  twistLimitDegrees=45
  twistMotion=Free
 }
 {
  bone1Name=neck_02
  bone2Name=spine_05
  swing1LimitDegrees=45
  swing1Motion=Free
  swing2LimitDegrees=45
  swing2Motion=Free
  twistLimitDegrees=45
  twistMotion=Free
 }
]
